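; Creative vehicle spawner device for the Shopping Cart, version 2, adjusting each cart it spawns.
; When a vehicle spawns, it reads the vehicle's configs through the cached vehicle interface and zeroes the maximum time at zero health.
; It then passes the spawned vehicle on to the parent spawner logic; the construction script only calls its parent.
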(event EventOnVehicleSpawned
  (params (Spawned%20Vehicle :type Fort%20Athena%20Vehicle%20Object%20Reference))
  (Variables|/Script/FortniteGame.FortVehicleConfigs|SetMaxTimeZeroHealth 0.0 (Class|FortVehicleInterface|GetVehicleConfigs (Variables|Default|GetCachedVehicleInterface)))
  (|CallParentFunction Spawned%20Vehicle))

(fn ConstructionScript
  (|CallParentFunction))
